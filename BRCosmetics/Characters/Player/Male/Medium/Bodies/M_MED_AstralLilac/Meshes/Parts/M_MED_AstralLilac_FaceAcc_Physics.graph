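# Physics asset for the face accessory mesh of the male medium AstralLilac player cosmetic, covering its dynamic hair bones.
# Small capsule bodies cover left, middle and right hair bone groups, and a larger tilted capsule covers the head.
# Each hair body is constrained directly to the head rather than chained to its neighbor.
# Narrow swing and twist limits apply, with tighter twist on the middle group.
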PhysicsAsset /BRCosmetics/Characters/Player/Male/Medium/Bodies/M_MED_AstralLilac/Meshes/Parts/M_MED_AstralLilac_FaceAcc_Physics
bodies {
 dyn_hair_01_l [{shapeType=Capsule, length=1, radius=3, shapeName=dyn_hair_01_l_capsule, center=<0 -2.9705 0>}]
 dyn_hair_01_mid [{shapeType=Capsule, length=1, radius=3, shapeName=dyn_hair_01_mid_capsule, center=<0 -2 0>}]
 dyn_hair_01_r [{shapeType=Capsule, length=1, radius=3, shapeName=dyn_hair_01_r_capsule, center=<0 2.9705 0>}]
 dyn_hair_02_l [{shapeType=Capsule, length=1, radius=3, shapeName=dyn_hair_02_l_capsule, center=<0 -2 0>}]
 dyn_hair_02_mid [{shapeType=Capsule, length=1, radius=3, shapeName=dyn_hair_02_mid_capsule, center=<0 -2.4033 0>}]
 dyn_hair_02_r [{shapeType=Capsule, length=1, radius=3, shapeName=dyn_hair_02_r_capsule, center=<0 2 0>}]
 dyn_hair_03_l [{shapeType=Capsule, length=1, radius=3, shapeName=dyn_hair_03_l_capsule, center=<0 -2 0>}]
 dyn_hair_03_mid [{shapeType=Capsule, length=1, radius=3, shapeName=dyn_hair_03_mid_capsule, center=<0 -2 0>}]
 dyn_hair_03_r [{shapeType=Capsule, length=1, radius=3, shapeName=dyn_hair_03_r_capsule, center=<0 2 0>}]
 dyn_hair_04_mid [{shapeType=Capsule, length=1, radius=3, shapeName=dyn_hair_04_mid_capsule, center=<0 -2 0>}]
 dyn_hair_05_mid [{shapeType=Capsule, length=1, radius=3, shapeName=dyn_hair_05_mid_capsule, center=<0 -2 0>}]
 head [{shapeType=Capsule, length=5.0853, radius=8.2826, center=<6.5161 2.5137 0>, rotation=<rot 90 -28.1197 0>}]
}
constraints [
 {bone1Name=dyn_hair_01_l, bone2Name=head, swing1LimitDegrees=20, swing2LimitDegrees=20, twistLimitDegrees=20}
 {bone1Name=dyn_hair_01_mid, bone2Name=head, swing1LimitDegrees=10, swing2LimitDegrees=25, twistLimitDegrees=15}
 {bone1Name=dyn_hair_01_r, bone2Name=head, swing1LimitDegrees=20, swing2LimitDegrees=20, twistLimitDegrees=20}
 {bone1Name=dyn_hair_02_l, bone2Name=head, swing1LimitDegrees=20, swing2LimitDegrees=20, twistLimitDegrees=20}
 {bone1Name=dyn_hair_02_mid, bone2Name=head, swing1LimitDegrees=25, swing2LimitDegrees=25, twistLimitDegrees=15}
 {bone1Name=dyn_hair_02_r, bone2Name=head, swing1LimitDegrees=20, swing2LimitDegrees=20, twistLimitDegrees=20}
 {bone1Name=dyn_hair_03_l, bone2Name=head, swing1LimitDegrees=20, swing2LimitDegrees=20, twistLimitDegrees=20}
 {bone1Name=dyn_hair_03_mid, bone2Name=head, swing1LimitDegrees=25, swing2LimitDegrees=25, twistLimitDegrees=15}
 {bone1Name=dyn_hair_03_r, bone2Name=head, swing1LimitDegrees=20, swing2LimitDegrees=20, twistLimitDegrees=20}
 {bone1Name=dyn_hair_04_mid, bone2Name=head, swing1LimitDegrees=25, swing2LimitDegrees=25, twistLimitDegrees=15}
 {bone1Name=dyn_hair_05_mid, bone2Name=head, swing1LimitDegrees=15, swing2LimitDegrees=20, twistLimitDegrees=15}
]
