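# Animation Blueprint for the AstralLilac character's face accessory, adding secondary physics and material-driven cloth motion.
# It copies the parent mesh pose, then runs a zero-gravity RigidBody simulation in head bone space.
# Each update feeds global and three regional WPO functions on base and overlay dynamic materials, using neck velocity and head collision; region 2 receives zero delta time.
# Skydiving or an open parachute switches on cloth wind and turbulence parameters, otherwise zeroed.
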AnimBlueprint /BRCosmetics/Characters/Player/Male/Medium/Bodies/M_MED_AstralLilac/Meshes/Parts/M_MED_AstralLilac_FaceAcc_AnimBP
target_skeleton=&/M_MED_AstralLilac_FaceAcc_Skeleton
graphs [
 AnimGraph {nodes [
   n1 = RigidBody!(
    ComponentPose=Local To Component(LocalPose=Copy Pose From Mesh() {bUseAttachedParent=true})
    Alpha
    OverrideWorldGravity:Vector="0,0,0"
    ExternalForce
   ) -> (Pose) {
    bClampLinearTranslationLimitToRefPose=true
    bOverrideWorldGravity=true
    overridePhysicsAsset=&/M_MED_AstralLilac_FaceAcc_Physics
    simulationSpace=BaseBoneSpace
    alphaBoolBlend {blendOption=Linear}
    baseBoneRef {boneName=head}
    simSpaceSettings {maxAngularAcceleration=100, maxAngularVelocity=3, maxLinearAcceleration=100, maxLinearVelocity=100, velocityScaleZ=0, worldAlpha=0.9}
   }
   Output Pose(Result=Component To Local(ComponentPose=@n1)) {layerGroup=None}
  ]}
 EventGraph {nodes [
   n1 = Event BlueprintUpdateAnimation()
   > n2 = Sequence()
   > Is Valid(InputObject=Get Player Pawn() -> (Player Pawn:Fort Player Pawn Athena Object Reference))
   > Set WPO Globals Function(
    Material=Get Material Instance Dynamic() -> (Material Instance Dynamic:Material Instance Dynamic Object Reference)
    Overlay Material=Get Overlay Material Instance Dynamic() -> (Overlay Material Instance Dynamic:Material Instance Dynamic Object Reference)
    Skeletal Mesh=Get Skeletal Mesh() -> (Skeletal Mesh:Skeletal Mesh Component Object Reference)
    "Local Noise Transform  Location"="Get Local Noise Transform  Location"() -> ("Local Noise Transform  Location":Vector)
    Local Noise Offset=Get Local Noise Offset()
   )
   n3 = Get Skeletal Mesh()
   n4 = Get Velocity Interp Speed()
   n5 = Get Overlay Material Instance Dynamic()
   n6 = Get Material Instance Dynamic()
   > Set WPO Regional Function!(
    execute
    Material:Material Instance Dynamic Object Reference=@n6
    Overlay Material:Material Instance Dynamic Object Reference=@n5
    Skeletal Mesh:Skeletal Mesh Component Object Reference=@n3
    Velocity Transform:Name=neck_01
    Velocity Interp Speed:double=@n4
    Collision Transform:Name=head
    Collision Transform Axis:EAxis Enum
    Invert Collision Transform Axis:bool=false
    Delta Time X:double=@n1.DeltaTimeX
    Previous Velocity Transform Location:"Vector (by ref)"=Get R1 Velocity Transform Location()
    Previous Velocity Transform Velocity:"Vector (by ref)"=Get R1 Velocity Transform Velocity()
    Region:int=1
   ) -> (then)
   > Set WPO Regional Function(
    Material=@n6
    Overlay Material=@n5
    Skeletal Mesh=@n3
    Velocity Transform=neck_01
    Velocity Interp Speed=@n4
    Collision Transform=head
    Invert Collision Transform Axis=false
    Delta Time X=0
    Previous Velocity Transform Location=Get R2 Velocity Transform Location()
    Previous Velocity Transform Velocity=Get R2 Velocity Transform Velocity()
    Region=2
   )
   > Set WPO Regional Function(
    Material=@n6
    Overlay Material=@n5
    Skeletal Mesh=@n3
    Velocity Transform=neck_01
    Velocity Interp Speed=@n4
    Collision Transform=head
    Invert Collision Transform Axis=false
    Delta Time X=@n1.DeltaTimeX
    Previous Velocity Transform Location=Get R3 Velocity Transform Location() -> (R3 Velocity Transform Location:Vector)
    Previous Velocity Transform Velocity=Get R3 Velocity Transform Velocity() -> (R3 Velocity Transform Velocity:Vector)
    Region=3
   )
   > n7 = Set isWPOEnabled(isWPOEnabled:bool=OR Boolean!(A=Get bIsSkydiveFloating(), B=Get bIsSkydiveDiveMode(), C:bool=Get bIsParachuteOpened()) -> (ReturnValue)) -> (Output_Get:bool)
   > n8 = Branch(Condition=@n7)
   > Set WPO Pho Wind Intensity(WPO Pho Wind Intensity=1)
   > n9 = Set WPO Pho TurbulenceNoise(WPO Pho TurbulenceNoise:double=65) -> (Output_Get:double)
   n10 = Get Overlay Material Instance Dynamic()
   n11 = Get Material Instance Dynamic()
   SetScalarParameterValue(execute=[@n9.then, @n12.then], self=[@n10, @n11], ParameterName=ClothSim_WindStrength, Value=Get WPO Pho Wind Intensity())
   > SetScalarParameterValue(
    self=[@n10, @n11]
    ParameterName=ClothSim_TurbulenceNoiseTileDivisor
    Value=Get WPO Pho TurbulenceNoise() -> (WPO Pho TurbulenceNoise:double)
   )
   Set WPO Pho Wind Intensity(execute=@n8.else, WPO Pho Wind Intensity=0)
   > n12 = Set WPO Pho TurbulenceNoise(WPO Pho TurbulenceNoise=0)
   Gravity Override State Function DEPRECATED(
    execute=@n2.then_1
    CharacterPartAnimInstance=Self-Reference()
    GameTimeInSeconds=GetGameTimeInSeconds()
    GravityOverrideParams=Get Gravity Override Params()
    DebugJointName=Pelvis
   )
   > Set Gravity Override(Gravity Override="0, 0, 0")
   Event BlueprintInitializeAnimation()
   > Set Skeletal Mesh(Skeletal Mesh:Skeletal Mesh Component Object Reference=GetOwnerSkeletalMeshComponent()) -> (Output_Get:Skeletal Mesh Component Object Reference)
   n13 = TryGetPawnOwner()
   > Is Valid(InputObject=@n13)
   > n14 = Cast To FortPlayerPawnAthena(Object=@n13)
   > Set Player Pawn(Player Pawn:Fort Player Pawn Athena Object Reference=@n14) -> (Output_Get:Fort Player Pawn Athena Object Reference)
   Event OnCharacterCustomizationLoaded()
   > n15 = Set Skeletal Mesh(Skeletal Mesh=GetPartSkeletalMeshComponent())
   > n16 = Get Dynamic Material Instance(SkeletalMeshComponent=@n15)
   > Set Material Instance Dynamic(Material Instance Dynamic:Material Instance Dynamic Object Reference=@n16) -> (Output_Get:Material Instance Dynamic Object Reference)
   > n17 = GetOrCreateDynamicOverlayMaterial(MeshComponent=@n15, MaterialName=MI_AstralLilac_FaceAcc)
   > Set Overlay Material Instance Dynamic(Overlay Material Instance Dynamic:Material Instance Dynamic Object Reference=@n17) -> (Output_Get:Material Instance Dynamic Object Reference)
   Get Gravity Override Params()
  ]}
]
variables [
 Gravity Override {type=Vector}
 Gravity Override Params {type=GravityOverrideParamsStruct, default {skydive_FloatingDown {gravityOverride=<0 -700 -600>}}}
 isWPOEnabled {type=boolean}
 Local Noise Offset {type=Vector}
 "Local Noise Transform  Location" {type=Vector}
 Material Instance Dynamic {type=MaterialInstanceDynamic}
 Overlay Material Instance Dynamic {type=MaterialInstanceDynamic}
 Player Pawn {type=FortPlayerPawnAthena}
 R1 Velocity Transform Location {type=Vector}
 R1 Velocity Transform Velocity {type=Vector}
 R2 Velocity Transform Location {type=Vector}
 R2 Velocity Transform Velocity {type=Vector}
 R3 Velocity Transform Location {type=Vector}
 R3 Velocity Transform Velocity {type=Vector}
 Skeletal Mesh {type=SkeletalMeshComponent}
 Velocity Interp Speed {type=number}
 WPO Pho TurbulenceNoise {type=number}
 WPO Pho Wind Intensity {type=number}
]
